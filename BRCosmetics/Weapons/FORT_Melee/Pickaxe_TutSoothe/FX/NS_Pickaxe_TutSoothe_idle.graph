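# Niagara idle effect for the TutSoothe pickaxe cosmetic: a pulsing, translucent teal glow offset up from the system origin.
# A single emitter burst-spawns a tall sprite with Z-axis alignment, drawn by sprite and light renderers.
# The particle loops its lifetime instead of dying, so width and alpha curves repeatedly swell to a midpoint peak and fall back.
# Front-end and preview user booleans are exposed, with no visible use in the emitter.
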NiagaraSystem /BRCosmetics/Weapons/FORT_Melee/Pickaxe_TutSoothe/FX/NS_Pickaxe_TutSoothe_idle
dependencies {
 usedDataInterfaces [&NiagaraDataInterfaceCurve, &NiagaraDataInterfaceVector2DCurve]
 usedDynamicInputs []
 usedModules [&SystemState]
}
emitters {Minimal {
  emitterUpdateScript [
   &EmitterState(
    Inactive Response=::NewEnumerator2
    Life Cycle Mode=::NewEnumerator1
    Loop Behavior=::NewEnumerator1
    Loop Duration Mode=::NewEnumerator1
    UseLoopDelay=false
   )
   &SpawnBurst_Instantaneous(Use Spawn Probability=false, UseLoopCountLimit=false)
  ]
  particleSpawnScript [
   &InitializeParticle(
    Color=<rgba 0 1 0.7723 0.15>
    Lifetime=5
    Sprite Size=<15 44>
    Sprite Size Mode=::NewEnumerator4
    UsePositionOffset=false
    Write Material Random=true
   )
   &SystemLocation(Offset=<0 0.5 38>)
  ]
  particleUpdateScript [
   &ParticleState(
    Kill Particles When Lifetime Has Elapsed=false
    Let Infinitely Lived Particles Die When Emitter Deactivates=true
    Loop Particles Lifetime=true
   )
   &ScaleSpriteSize(
    Scale Factor=&Vector2DFromCurve(
     Vector2Curve=DataInterface() {propertyValues {
       XCurve {
        defaultValue=1
        keys [
         {arriveTangent=-0.0031, interpMode=RCIM_Cubic, leaveTangent=-0.0031, leaveTangentWeight=0.1674, tangentMode=RCTM_User, time=0, value=0.85}
         {
          arriveTangent=0.0012
          arriveTangentWeight=0.1667
          interpMode=RCIM_Cubic
          leaveTangent=0.0012
          leaveTangentWeight=0.1667
          tangentMode=RCTM_User
          time=0.5
          value=1
         }
         {arriveTangent=0.0035, arriveTangentWeight=0.1668, interpMode=RCIM_Cubic, leaveTangent=0.0035, tangentMode=RCTM_User, time=1, value=0.85}
        ]
       }
       YCurve {keys [<at 0 1>,<at 1 1>]}
      }}
     ScaleCurve=<1 1>
    )
    Scale Sprite Size Mode=::NewEnumerator2
   )
   &ScaleColor(
    Scale Alpha=&FloatFromCurve(FloatCurve=DataInterface() {propertyValues {Curve {
        defaultValue=1
        keys [
         {arriveTangent=-0.0031, interpMode=RCIM_Cubic, leaveTangent=-0.0031, tangentMode=RCTM_User, time=0, value=0.65}
         {arriveTangent=0.0012, arriveTangentWeight=0.1667, interpMode=RCIM_Cubic, leaveTangent=0.0012, tangentMode=RCTM_User, time=0.5, value=1}
         {arriveTangent=0.0035, arriveTangentWeight=0.1667, interpMode=RCIM_Cubic, leaveTangent=0.0035, tangentMode=RCTM_User, time=1, value=0.65}
        ]
       }}})
    ScaleA=true
    ScaleRGB=false
   )
   &SpriteFacingAndAlignment(Sprite Alignment=<0 0 1>, Sprite Facing=<0 0 -1>, WriteAlignment=true, WriteFacing=false)
  ]
  rendererClasses [&NiagaraSpriteRendererProperties, &NiagaraLightRendererProperties]
  renderers [
   {rendererClass=&NiagaraSpriteRendererProperties}
   {rendererClass=&NiagaraLightRendererProperties, rendererIndex=1}
  ]
 }}
summary {
 systemName=NS_Pickaxe_TutSoothe_idle
 emitters [{emitterName=Minimal, rendererClasses [&NiagaraSpriteRendererProperties, &NiagaraLightRendererProperties]}]
 userVariables [
  User.bIsFrontEnd {type=NiagaraBool}
  User.bIsFrontEndPreview {type=NiagaraBool}
 ]
}
